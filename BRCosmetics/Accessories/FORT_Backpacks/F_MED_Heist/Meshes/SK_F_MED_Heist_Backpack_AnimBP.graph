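# Animation blueprint for the F_MED_Heist backpack that adds physics motion to its money bill bones, including flutter while skydiving.
# Seven chained AnimDynamics nodes (bills_01 to 07) spring the incoming pose within tight limits, each pushed along X or Y.
# The update event derives that force from Perlin noise of pawn height, zero outside skydiving and scaled down under parachute.
# A duplicate noise branch is left unconnected.
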AnimBlueprint /BRCosmetics/Accessories/FORT_Backpacks/F_MED_Heist/Meshes/SK_F_MED_Heist_Backpack_AnimBP
target_skeleton=&/SK_F_MED_Heist_Backpack_Skeleton
graphs [
 AnimGraph {nodes [
   AnimGraphNode_SubInput_0 = Input Pose()
   n1 = Get falling_external_force_y() -> (falling_external_force_y:Vector)
   n2 = Anim Dynamics - Bone: bills_01(ComponentPose=Local To Component(LocalPose=@AnimGraphNode_SubInput_0), Alpha=1, ExternalForce=@n1) {
    angularDampingOverride=0.99
    angularSpringConstant=350
    bAngularSpring=true
    bGravityOverrideInSimSpace=true
    bOverrideAngularDamping=true
    gravityScale=0.15
    lODThreshold=1
    linearDampingOverride=0.99
    linearSpringConstant=200
    numSolverIterationsPostUpdate=4
    numSolverIterationsPreUpdate=16
    alphaBoolBlend {blendOption=Linear}
    boundBone {boneName=bills_01}
    componentAppliedLinearAccClamp=<10000 10000 10000>
    componentLinearAccScale=<1 1 1>
    componentLinearVelScale=<1 1 1>
    physicsBodyDefinitions [{
      sphereCollisionRadius=0
      boundBone {boneName=bills_01}
      boxExtents=<20 20 1>
      constraintSetup {
       angularTargetAxis=AxisY
       twistAxis=AxisY
       angularLimitsMax=<45 0 5>
       angularLimitsMin=<-5 0 -5>
       angularTarget=<0 0 0>
       linearAxesMax=<0 0 1>
       linearAxesMin=<0 0 -1>
      }
      localJointOffset=<0 0 -2>
     }]
   }
   n3 = Get falling_external_force_X() -> (falling_external_force_X:Vector)
   n4 = Anim Dynamics - Bone: bills_02(ComponentPose=@n2, Alpha=1, ExternalForce=@n3) {
    angularDampingOverride=0.99
    angularSpringConstant=350
    bAngularSpring=true
    bGravityOverrideInSimSpace=true
    bOverrideAngularDamping=true
    gravityScale=0.15
    lODThreshold=1
    linearDampingOverride=0.99
    linearSpringConstant=200
    numSolverIterationsPostUpdate=4
    numSolverIterationsPreUpdate=16
    alphaBoolBlend {blendOption=Linear}
    boundBone {boneName=bills_02}
    componentAppliedLinearAccClamp=<10000 10000 10000>
    componentLinearAccScale=<1 1 1>
    componentLinearVelScale=<1 1 1>
    physicsBodyDefinitions [{
      sphereCollisionRadius=0
      boundBone {boneName=bills_02}
      boxExtents=<20 20 1>
      constraintSetup {
       angularTargetAxis=AxisY
       twistAxis=AxisY
       angularLimitsMax=<0 0 10>
       angularLimitsMin=<0 0 -20>
       angularTarget=<0 0 0>
       linearAxesMax=<0 0 1>
       linearAxesMin=<0 0 -1>
      }
      localJointOffset=<0 0 -1>
     }]
   }
   n5 = Anim Dynamics - Bone: bills_03(ComponentPose=@n4, Alpha=1, ExternalForce=@n1) {
    angularDampingOverride=0.99
    angularSpringConstant=350
    bAngularSpring=true
    bGravityOverrideInSimSpace=true
    bOverrideAngularDamping=true
    gravityScale=0.15
    lODThreshold=1
    linearDampingOverride=0.99
    linearSpringConstant=200
    numSolverIterationsPostUpdate=4
    numSolverIterationsPreUpdate=16
    alphaBoolBlend {blendOption=Linear}
    boundBone {boneName=bills_03}
    componentAppliedLinearAccClamp=<10000 10000 10000>
    componentLinearAccScale=<1 1 1>
    componentLinearVelScale=<1 1 1>
    physicsBodyDefinitions [{
      sphereCollisionRadius=0
      boundBone {boneName=bills_03}
      boxExtents=<20 20 1>
      constraintSetup {
       angularTargetAxis=AxisY
       twistAxis=AxisY
       angularLimitsMax=<0 0 10>
       angularLimitsMin=<0 0 -20>
       angularTarget=<0 0 0>
       linearAxesMax=<0 0 1>
       linearAxesMin=<0 0 -1>
      }
      localJointOffset=<0 0 -1>
     }]
   }
   n6 = Anim Dynamics - Bone: bills_04(ComponentPose=@n5, Alpha=1, ExternalForce=@n1) {
    angularDampingOverride=0.99
    bAngularSpring=true
    bGravityOverrideInSimSpace=true
    bOverrideAngularDamping=true
    gravityScale=0.15
    lODThreshold=1
    linearDampingOverride=0.99
    numSolverIterationsPostUpdate=4
    numSolverIterationsPreUpdate=16
    alphaBoolBlend {blendOption=Linear}
    boundBone {boneName=bills_04}
    componentAppliedLinearAccClamp=<10000 10000 10000>
    componentLinearAccScale=<1 1 1>
    componentLinearVelScale=<1 1 1>
    physicsBodyDefinitions [{
      sphereCollisionRadius=0
      boundBone {boneName=bills_04}
      boxExtents=<20 20 1>
      constraintSetup {
       angularTargetAxis=AxisY
       twistAxis=AxisY
       angularLimitsMax=<30 30 5>
       angularLimitsMin=<-5 -5 -5>
       angularTarget=<0 0 0>
       linearAxesMax=<0 0 1>
       linearAxesMin=<0 0 -1>
      }
      localJointOffset=<0 0 -1>
     }]
   }
   n7 = Anim Dynamics - Bone: bills_05(ComponentPose=@n6, Alpha=1, ExternalForce=@n1) {
    angularDampingOverride=0.99
    bAngularSpring=true
    bGravityOverrideInSimSpace=true
    bOverrideAngularDamping=true
    gravityScale=0.15
    lODThreshold=1
    linearDampingOverride=0.99
    numSolverIterationsPostUpdate=4
    numSolverIterationsPreUpdate=16
    alphaBoolBlend {blendOption=Linear}
    boundBone {boneName=bills_05}
    componentAppliedLinearAccClamp=<10000 10000 10000>
    componentLinearAccScale=<1 1 1>
    componentLinearVelScale=<1 1 1>
    physicsBodyDefinitions [{
      sphereCollisionRadius=0
      boundBone {boneName=bills_05}
      boxExtents=<20 20 1>
      constraintSetup {
       angularTargetAxis=AxisY
       twistAxis=AxisY
       angularLimitsMax=<5 5 5>
       angularLimitsMin=<-30 -30 -5>
       angularTarget=<0 0 0>
       linearAxesMax=<0 0 1>
       linearAxesMin=<0 0 -1>
      }
      localJointOffset=<0 0 -1>
     }]
   }
   n8 = Anim Dynamics - Bone: bills_06(ComponentPose=@n7, Alpha=1, ExternalForce=@n3) {
    angularDampingOverride=0.99
    angularSpringConstant=300
    bAngularSpring=true
    bGravityOverrideInSimSpace=true
    bOverrideAngularDamping=true
    gravityScale=0.15
    lODThreshold=1
    linearDampingOverride=0.99
    linearSpringConstant=200
    numSolverIterationsPostUpdate=4
    numSolverIterationsPreUpdate=16
    alphaBoolBlend {blendOption=Linear}
    boundBone {boneName=bills_06}
    componentAppliedLinearAccClamp=<10000 10000 10000>
    componentLinearAccScale=<1 1 1>
    componentLinearVelScale=<1 1 1>
    physicsBodyDefinitions [{
      sphereCollisionRadius=0
      boundBone {boneName=bills_06}
      boxExtents=<20 17 1>
      constraintSetup {
       angularTargetAxis=AxisY
       twistAxis=AxisY
       angularLimitsMax=<30 0 5>
       angularLimitsMin=<-5 0 -5>
       angularTarget=<0 0 0>
       linearAxesMax=<0 0 1>
       linearAxesMin=<0 0 -1>
      }
      localJointOffset=<0 0 -1>
     }]
   }
   n9 = Anim Dynamics - Bone: bills_07(ComponentPose=@n8, Alpha=1, ExternalForce=@n3) {
    angularDampingOverride=0.99
    angularSpringConstant=300
    bAngularSpring=true
    bGravityOverrideInSimSpace=true
    bOverrideAngularDamping=true
    gravityScale=0.15
    lODThreshold=1
    linearDampingOverride=0.99
    linearSpringConstant=200
    numSolverIterationsPostUpdate=4
    numSolverIterationsPreUpdate=16
    alphaBoolBlend {blendOption=Linear}
    boundBone {boneName=bills_07}
    componentAppliedLinearAccClamp=<10000 10000 10000>
    componentLinearAccScale=<1 1 1>
    componentLinearVelScale=<1 1 1>
    physicsBodyDefinitions [{
      sphereCollisionRadius=0
      boundBone {boneName=bills_07}
      boxExtents=<17 20 1>
      constraintSetup {
       angularTargetAxis=AxisY
       twistAxis=AxisY
       angularLimitsMax=<30 0 5>
       angularLimitsMin=<-5 0 -5>
       angularTarget=<0 0 0>
       linearAxesMax=<0 0 1>
       linearAxesMin=<0 0 -1>
      }
      localJointOffset=<0 0 -1.2>
     }]
   }
   Output Pose(Result=Component To Local(ComponentPose=@n9)) {layerGroup=None}
  ]}
 EventGraph {nodes [
   Event BlueprintUpdateAnimation()
   n1 = Get PawnLastLocation() -> (PawnLastLocation:Vector)
   n2 = SelectFloat(
    A=Make Perlin Noise and Remap!(Value:float=BreakVector(InVec=@n1).Z, RangeOutMin:float=-5000, RangeOutMax:float=5000) -> (ReturnValue:float)
    B=0
    bPickA=Get bIsSkyDiving()
   )
   n3 = SelectFloat(A=float * float(A=@n2, B=0.2), B=@n2, bPickA=Get bIsParachuteOpened())
   > Set falling_external_force_y(falling_external_force_y:Vector=MakeVector(Y=@n3)) -> (Output_Get:Vector)
   > Set falling_external_force_X(falling_external_force_X:Vector=MakeVector(X=@n3, Y=0)) -> (Output_Get:Vector)
   n4 = SelectFloat(A=Make Perlin Noise and Remap(Value=0, RangeOutMin=-5000, RangeOutMax=5000), B=0, bPickA=Get bIsSkyDiving())
   SelectFloat(A=float * float(A=@n4, B=0.2), B=@n4, bPickA=Get bIsParachuteOpened())
  ]}
]
variables [
 falling_external_force_X {type=Vector}
 falling_external_force_y {type=Vector}
]
